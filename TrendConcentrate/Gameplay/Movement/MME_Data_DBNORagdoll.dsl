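; Data asset for a movement element named DBNORagdoll, holding the state of a down-but-not-out ragdoll.
; It stores remaining ragdoll time, the original collision profile, the original mesh relative transform and the current bump strength.
; Its only logic is a reset event that returns these four variables to fixed defaults, clearing the saved values and restoring a non-zero bump strength.
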(event EventResettoDefaultValues
  (Variables|Default|SetRemainingRagdollTIme 0.0)
  (Variables|Default|SetOriginalCollisionProfile "None")
  (Variables|Default|SetOriginalMeshRelativeTransform)
  (Variables|Default|SetCurrentBumpStrength 2e+02))
