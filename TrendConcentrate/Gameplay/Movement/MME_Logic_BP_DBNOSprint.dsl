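; Movement mode extension logic that lets a downed (DBNO) player sprint, activating only while the owner pawn is DBNO.
; Activation adds speedline and camera-shake cues plus a tactical sprint gameplay effect; deactivation removes them and stops energy use.
; Each update drains energy by the DBNO behavior override component's stamina cost times delta time, ending once energy is insufficient.
; Max speed is scaled by a ScalableFloat multiplier.
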(event EventOnExtensionActivate
  (params (RuntimeContext :type MMERuntime%20Context%20Structure%20%28by%20ref%29))
  (|CallParentFunction RuntimeContext)
  (Development|PrintString 0 "Activate DBNO Sprint")
  (bind _returnvalue (Class|FortMovementModeExtBlueprintLibrary|GetRuntimeDatafromContext RuntimeContext "/TrendConcentrate/Gameplay/Movement/MME_RuntimeData_BP_DBNOSprint.MME_RuntimeData_BP_DBNOSprint_C"))
  (Variables|Default|SetRuntimeData _returnvalue)
  (GameplayCue|AddGameplayCueOnActor(Looping) (Variables|FortMovementMode_BaseExtLogic|GetOwnerPawn) (Variables|Default|GetSpeedlinesCue) (Variables|Default|GetParameters))
  (GameplayCue|AddGameplayCueOnActor(Looping) (Variables|FortMovementMode_BaseExtLogic|GetOwnerPawn) (Variables|Default|GetCamShakeCue) (Variables|Default|GetParameters))
  (bind _returnvalue_2 (GameplayEffects|ApplyGameplayEffectToSelf :self (Ability|GetAbilitySystemComponent (Variables|FortMovementMode_BaseExtLogic|GetOwnerPawn)) :GameplayEffectClass "/TrendConcentrate/Gameplay/Effects/GE_Athena_TacticalSprint_DBNO.GE_Athena_TacticalSprint_DBNO_C" :Level 0.0))
  (Variables|Default|SetTacSprintGE _returnvalue_2))

(event EventOnExtensionDeactivate
  (params (RuntimeContext :type MMERuntime%20Context%20Structure%20%28by%20ref%29))
  (bind _parameters (Variables|Default|GetParameters))
  (|CallParentFunction RuntimeContext)
  (Development|PrintString 0 "Deactivate DBNO Sprint")
  (Energy|StopUsingEnergy (Actor|GetComponentbyClass (Variables|FortMovementMode_BaseExtLogic|GetOwnerPawn) "/Script/EnergyRuntime.FortComponent_Energy") (Variables|Default|GetEnergy%20Use%20Identifier) (Variables|FortMovementMode_BaseExtLogic|GetOwnerPawn))
  (GameplayCue|RemoveGameplayCueOnActor(Looping) (Variables|FortMovementMode_BaseExtLogic|GetOwnerPawn) (Variables|Default|GetSpeedlinesCue) _parameters)
  (GameplayCue|RemoveGameplayCueOnActor(Looping) (Variables|FortMovementMode_BaseExtLogic|GetOwnerPawn) (Variables|Default|GetCamShakeCue) _parameters)
  (GameplayEffects|RemoveActiveGameplayEffect (Ability|GetAbilitySystemComponent (Variables|FortMovementMode_BaseExtLogic|GetOwnerPawn)) (Variables|Default|GetTacSprintGE) -1))

(event EventUpdateAfterCharacterMovement
  (params (RuntimeContext :type MMERuntime%20Context%20Structure%20%28by%20ref%29) (OldLocation :type Vector%20%28by%20ref%29 :default "0, 0, 0") (OldVelocity :type Vector%20%28by%20ref%29 :default "0, 0, 0")))

(fn CanActivateExtension
  (return (Health|IsDBNO (Variables|FortMovementMode_BaseExtLogic|GetOwnerPawn))))

(fn GetModifiedBaseMaxSpeed
  (params (BaseMaxSpeed :type Float%20%28single-precision%29) (RuntimeContext :type MMERuntime%20Context%20Structure%20%28by%20ref%29))
  (return (* BaseMaxSpeed (ScalableFloat|GetValueatLevel (Variables|Default|GetSprintSpeedMultiplyer_SF)))))

(fn PhysUpdate
  (params (RuntimeContext :type MMERuntime%20Context%20Structure%20%28by%20ref%29) (InVelocity :type Vector%20%28by%20ref%29 :default "0, 0, 0") (InAcceleration :type Vector%20%28by%20ref%29 :default "0, 0, 0"))
  (return "0, 0, 0" "0, 0, 0"))

(fn UpdateBeforeCharacterMovement
  (params (RuntimeContext :type MMERuntime%20Context%20Structure%20%28by%20ref%29))
  (Energy|UseEnergy (Actor|GetComponentbyClass (Variables|FortMovementMode_BaseExtLogic|GetOwnerPawn) "/Script/EnergyRuntime.FortComponent_Energy") (* (DBNO|GetDBNOSprintingStaminaCost (Actor|GetComponentbyClass (Variables|FortMovementMode_BaseExtLogic|GetOwnerPawn) "/TrendConcentrate/Components/B_DBNOBehaviorOverrideComponent_TrendConcentrate.B_DBNOBehaviorOverrideComponent_TrendConcentrate_C")) (Class|FortMovementModeExtBlueprintLibrary|GetDeltaTimefromContext RuntimeContext)) true (Variables|FortMovementMode_BaseExtLogic|GetOwnerPawn))
  (if (Energy|HasSufficientEnergy (Actor|GetComponentbyClass (Variables|FortMovementMode_BaseExtLogic|GetOwnerPawn) "/Script/EnergyRuntime.FortComponent_Energy") (Variables|Default|GetMinimumEnergy))
    (return "Active_Layered")
    (else
      (return "Ended"))))
