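; Movement mode extension logic that supplies a constant-force guided motion for a DBNO (down but not out) bump.
; Direction takes the owner's velocity with a slight downward bias, normalizes it, reflects it with the current floor hit, then adds an upward offset.
; The reflection uses the hit's impact point as its axis, not a surface normal.
; Strength is a fixed value, and the activation event is empty.
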(event EventOnExtensionActivate
  (params (RuntimeContext :type MMERuntime%20Context%20Structure%20%28by%20ref%29)))

(fn GetDirection_ConstantForce
  (params (GuidedMotionRuntimeData :type Fort%20Movement%20Mode%20Guided%20Motion%20Runtime%20Data%20Object%20Reference))
  (bind _ownermovementcomponent (Variables|FortMovementMode_BaseExtLogic|GetOwnerMovementComponent))
  (bind (_currentfloor_bblockinghit _currentfloor_bwalkablefloor _currentfloor_blinetrace _currentfloor_floordist _currentfloor_linedist _currentfloor_hitresult) (Utilities|Struct|BreakFindFloorResult (Variables|/Script/Engine.CharacterMovementComponent|GetCurrentFloor_bBlockingHit _ownermovementcomponent)))
  (bind (_currentfloor_hitresult_bblockinghit _currentfloor_hitresult_binitialoverlap _currentfloor_hitresult_time _currentfloor_hitresult_distance _currentfloor_hitresult_location _currentfloor_hitresult_impactpoint) (Collision|BreakHitResult _currentfloor_hitresult))
  (bind _output_get (Variables|Default|SetVelocityDirection (Math|Vector|Normalize (+ (Variables|/Script/Engine.MovementComponent|GetVelocity _ownermovementcomponent) "0.000000,0.000000,-1.000000"))))
  (bind _output_get_2 (LocalVariables|SetReflectedVector (- _output_get (* (* (Math|Vector|DotProduct (Variables|Default|GetVelocityDirection) _currentfloor_hitresult_impactpoint) 2.0) _currentfloor_hitresult_impactpoint))))
  (return (+ _output_get_2 "0.000000,0.000000,0.500000")))

(fn GetStrength_ConstantForce
  (params (GuidedMotionRuntimeData :type Fort%20Movement%20Mode%20Guided%20Motion%20Runtime%20Data%20Object%20Reference))
  (return 4e+01))
