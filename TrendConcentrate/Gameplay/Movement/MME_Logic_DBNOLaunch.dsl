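; MME_Logic_DBNOLaunch is movement mode extension logic for launching a pawn, by its name in the down-but-not-out state.
; Its constant-force direction override takes the owner pawn's normalized actor forward vector, stores it in the LaunchVector variable and returns it.
; The guided motion runtime data parameter is unused, so the direction depends only on the pawn's facing at the time of the call.
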(fn GetDirection_ConstantForce
  (params (GuidedMotionRuntimeData :type Fort%20Movement%20Mode%20Guided%20Motion%20Runtime%20Data%20Object%20Reference))
  (bind _output_get (LocalVariables|SetLaunchVector (Math|Vector|Normalize (Transformation|GetActorForwardVector (Variables|FortMovementMode_BaseExtLogic|GetOwnerPawn)))))
  (return _output_get))
