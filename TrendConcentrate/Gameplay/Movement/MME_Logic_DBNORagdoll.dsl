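; MME_Logic_DBNORagdoll is a movement extension logic asset, named for ragdoll behavior in the down-but-not-out (DBNO) state.
; It declares activate and deactivate events, each receiving an MME runtime context structure by reference.
; Both events are empty, so the file shows no ragdoll or movement logic of its own.
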(event EventOnExtensionActivate
  (params (RuntimeContext :type MMERuntime%20Context%20Structure%20%28by%20ref%29)))

(event EventOnExtensionDeactivate
  (params (RuntimeContext :type MMERuntime%20Context%20Structure%20%28by%20ref%29)))
